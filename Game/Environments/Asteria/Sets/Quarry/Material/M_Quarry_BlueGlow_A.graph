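# Deferred decal material for the Quarry set in Asteria that renders a pulsing blue emissive glow with alpha-composite blending.
# Vertex color and a noise texture mask an emissive color modulated by a time-driven sine.
# A three-input lerp driven by normalized camera distance zeroes the glow at both ends, fading it out far away to blend into HLOD.
# Low quality and ES3.1 switches compute an offset toward the object pivot, collapsing the geometry.
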Material /Game/Environments/Asteria/Sets/Quarry/Material/M_Quarry_BlueGlow_A
graph {
 properties {
  bUsedWithNanite=true
  blendMode=BLEND_AlphaComposite
  materialDomain=MD_DeferredDecal
  pixelDepthOffsetMode=PDOM_Legacy
  usedShadingModels=Default Lit
 }
 nodes [
  n1 = Subtract(A=ObjectPositionWS(), B=WorldPosition().XYZ)
  FeatureLevelSwitch(Default=QualitySwitch(Default=Constant3Vector(), Low=@n1), ES3_1=@n1)
  n2 = VertexColor()
  n3 = TextureSample() {texture=/Game/Environments/Asteria/Sets/Quarry/Textures/T_Asteria_Quarry_Noise}
  Multiply(A=@n2, B=@n3.RGB)
  n4 = Constant()
  n5 = Sine(None=Multiply(A=Time(), B=ScalarParameter() {defaultValue=0.504, parameterName=Frequency}))
  n6 = MaterialFunctionCall(
   1=@n4
   2=Multiply(A=VectorParameter() {parameterName=Emissive, defaultValue=<rgba 4.294 11.2313 25 1>}.RGB, B=LinearInterpolate(Alpha=@n5) {constA=0.8})
   3=@n4
   A=NamedRerouteUsage()
  ) {desc=Fade out far away to blend to HLOD, materialFunction=&Lerp_Multiple_Float3}
  Multiply(A=@n2, B=Multiply(A=@n3.RGB, B=@n6.Lerp 3 Inputs))
  Multiply(A=@n2, B=Multiply(A=@n3.RGB, B=@n6.Lerp 3 Inputs))
  n7 = Distance(A=WorldPosition().XYZ, B=CameraPositionWS())
  n8 = ScalarParameter() {defaultValue=50000, parameterName=Distance Fade}
  NamedRerouteDeclaration(None=Saturate(None=Divide(A=@n7, B=@n8))) {name=FadeVal, nodeColor=<rgba 0.5 1 0 1>}
 ]
}
