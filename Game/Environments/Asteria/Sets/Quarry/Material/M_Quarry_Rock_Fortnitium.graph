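# Translucent deferred decal material for Quarry rock fissure cracks, with diffuse, normal, specular and emissive maps and a glowing, cyan-tinted "Fortnitium" look.
# The emissive map is tinted, weighted by Fresnel and modulated by panning blurred noise.
# A static switch replaces mesh UVs with hard-edge world-aligned tiling.
# The same switch refines diffuse-alpha opacity with a spherical gradient, erosion and a decal-normal versus object-orientation fade.
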Material /Game/Environments/Asteria/Sets/Quarry/Material/M_Quarry_Rock_Fortnitium
graph {
 properties {blendMode=BLEND_Translucent, materialDomain=MD_DeferredDecal, pixelDepthOffsetMode=PDOM_Legacy, usedShadingModels=Default Lit}
 nodes [
  n1 = Panner(Time=Time(), Speed=ScalarParameter() {defaultValue=0.05, parameterName=Panner Speed})
  n2 = VectorParameter() {defaultValue=<rgba 0 2.2566 3 1>}
  n3 = ObjectOrientation()
  n4 = StaticSwitchParameter(
   True=Divide(
    A=MaterialFunctionCall!(UseLocalSpace, WorldPosition, NormalOverride=@n3) {materialFunction=/Game/Environments/Apollo/Sets/Skyfire/Materials/MaterialFunctions/MF_WorldAlignedUVs_HardEdge}.Result
    B=ScalarParameter() {defaultValue=1200, parameterName=WorldTilingDivisor}
   )
   False=TextureCoordinate()
  ) {parameterName="Us WS uvs "}
  n5 = TextureSampleParameter2D(UVs=@n4) {
   parameterName=Emissive
   samplerType=SAMPLERTYPE_LinearColor
   texture=/Game/Environments/Asteria/Sets/Quarry/Textures/T_Asteria_Quarry_FissureCrack_A_E
  }
  Multiply(
   A=LinearInterpolate(Alpha=TextureSample(UVs=@n1) {samplerType=SAMPLERTYPE_Grayscale, texture=&LowResBlurredNoise}.RGB) {constA=-0.1}
   B=LinearInterpolate(B=Multiply(A=@n2.RGB, B=@n5.RGB), Alpha=Fresnel() {baseReflectFraction=0.004, exponent=1}) {constB=0}
  )
  Reroute()
  n6 = TextureSampleParameter2D(UVs=@n4) {
   parameterName=Diffuse
   samplerType=SAMPLERTYPE_LinearColor
   texture=/Game/Environments/Asteria/Sets/Quarry/Textures/T_Asteria_Quarry_FissureCrack_A_D
  }
  n7 = ScalarParameter() {defaultValue=0.5, parameterName=SphericalGradient_Center_U}
  n8 = ScalarParameter() {defaultValue=0.5, parameterName=SphericalGradient_Center_V}
  n9 = Multiply(
   A=@n6.A
   B=SphereMask(
    A=TextureCoordinate()
    B=AppendVector(A=@n7, B=@n8)
    Radius=ScalarParameter() {defaultValue=0.5, parameterName=SphericalGradientRadius}
    Hardness=ScalarParameter() {defaultValue=0.6, parameterName=SphericalGradientHardness}
   )
  )
  n10 = ScalarParameter() {parameterName=Erosion}
  n11 = Saturate(None=DotProduct(A=MaterialFunctionCall() {materialFunction=/Game/Weapons/Prototype/SprayGun/Materials/MF_DecalNormal}.WorldNormal, B=ObjectOrientation()))
  StaticSwitchParameter(True=Multiply(A=Subtract(A=@n9, B=@n10), B=OneMinus(None=@n11)), False=@n6.A) {parameterName="Us WS uvs "}
  TextureSampleParameter2D(UVs=@n4) {
   parameterName=Normal
   samplerType=SAMPLERTYPE_Normal
   texture=/Game/Environments/Asteria/Sets/Quarry/Textures/T_Asteria_Quarry_FissureCrack_A_N
  }
  TextureSampleParameter2D(UVs=@n4) {
   parameterName=Specular
   samplerType=SAMPLERTYPE_LinearColor
   texture=/Game/Environments/Asteria/Sets/Quarry/Textures/T_Asteria_Quarry_FissureCrack_A_S
  }
 ]
}
